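# M_Proc_DebugAsset is a minimal Default Lit material among the debug assets of the procedural foundation art tools.
# Its graph holds a single Vertex Color node, with no textures, parameters or math, and no output connection is shown for it.
# The name and node suggest it serves to inspect vertex colors on procedural debug meshes.
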Material /Game/ArtTools/Procedural/Foundation/DebugAssets/M_Proc_DebugAsset
graph {properties {pixelDepthOffsetMode=PDOM_Legacy, usedShadingModels=Default Lit}, nodes [VertexColor()]}
